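# Translucent UI material for a circular progress bar with a three-color gradient fill and optional segments, delta, glow, outline and inner fill.
# Pixel-sized SDF circles form the ring, while radial-arc functions on animated start-to-end values produce progress and delta masks, optionally snapped to segments.
# Rotated circle SDFs add optional rounded caps, and a ping-pong mask cuts segment gaps.
# Named reroutes feed masks into color and opacity blends, using scalar lerps as feature toggles.
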Material /FortUILibrary/Systems/ProgressBar/Materials/M_UI_CircularProgressBar
graph {
 properties {
  bUsedWithStaticMesh=false
  blendMode=BLEND_Translucent
  materialDomain=MD_UI
  pixelDepthOffsetMode=PDOM_Legacy
  usedShadingModels=Default Lit
 }
 nodes [
  n1 = VectorParameter() {
   desc="Color for empty part of progress bar.\r\nAlpha defines opacity for empty part of progress bar."
   group=Color
   parameterName=EmptyColor
   sortPriority=10
   defaultValue=<rgba 0.3924 0.3924 0.3924 0.3>
  }
  n2 = ScalarParameter() {defaultValue=1, desc="Alpha for filled part of progress bar.\r\n0 -1", group=Color, parameterName=FillAlpha, sliderMax=1}
  n3 = LinearInterpolate(A=LinearInterpolate(B=@n1.A, Alpha=NamedRerouteUsage()), B=@n2, Alpha=Max(A=NamedRerouteUsage(), B=NamedRerouteUsage()))
  n4 = NamedRerouteUsage()
  n5 = ScalarParameter() {defaultValue=1, desc="Enable segments functionality.\r\nbool", group=Segments, parameterName=UseSegments, sliderMax=1, sortPriority=1}
  n6 = LinearInterpolate(B=LinearInterpolate(A=@n3, B=Multiply(A=@n3, B=@n4) {constA=1}, Alpha=@n5), Alpha=NamedRerouteUsage())
  n7 = VectorParameter() {group=Behind Circle, parameterName=BehindCircleColor, sortPriority=2, defaultValue=<rgba 0.1493 0.1493 0.1493 0.1>}
  n8 = ScalarParameter() {
   desc="Enables inner circle fill for the circular progress bar.\r\nbool"
   group=Behind Circle
   parameterName=UseBehindCircle
   sliderMax=1
   sortPriority=1
  }
  n9 = LinearInterpolate(A=@n6, B=LinearInterpolate(A=@n6, B=@n7.A, Alpha=NamedRerouteUsage()), Alpha=@n8)
  n10 = Multiply(A=VectorParameter() {group=Glow, parameterName=GlowColor, sortPriority=3, defaultValue=<rgba 0.1958 0 1 0.2>}.A, B=NamedRerouteUsage())
  n11 = ScalarParameter() {desc="Enables glow around progress fill part.\r\nbool", group=Glow, parameterName=UseProgressGlow, sliderMax=1, sortPriority=1}
  LinearInterpolate(A=@n9, B=Saturate(None=Add(A=@n9, B=@n10)), Alpha=@n11)
  n12 = VectorParameter() {
   desc="Color for empty part of progress bar.\r\nAlpha defines opacity for empty part of progress bar."
   group=Color
   parameterName=EmptyColor
   sortPriority=10
   defaultValue=<rgba 0.3924 0.3924 0.3924 0.3>
  }
  n13 = LinearInterpolate(B=@n12.RGB, Alpha=NamedRerouteUsage())
  n14 = VectorParameter() {group=Behind Circle, parameterName=BehindCircleColor, sortPriority=2, defaultValue=<rgba 0.1493 0.1493 0.1493 0.1>}
  n15 = ScalarParameter() {
   desc="Enables inner circle fill for the circular progress bar.\r\nbool"
   group=Behind Circle
   parameterName=UseBehindCircle
   sliderMax=1
   sortPriority=1
  }
  n16 = LinearInterpolate(A=@n13, B=LinearInterpolate(A=@n13, B=@n14.RGB, Alpha=NamedRerouteUsage()), Alpha=@n15)
  n17 = ScalarParameter() {
   desc="Enables GlowColor for the glow effect, otherwise will use progress fill color.\r\nbool"
   group=Glow
   parameterName=UseGlowColor
   sliderMax=1
   sortPriority=2
  }
  n18 = VectorParameter() {desc=Start color in gradient, group=Color, parameterName=ColorA, sortPriority=1, defaultValue=<rgba 0 0 1 1>}
  n19 = VectorParameter() {desc=Central color in gradient, group=Color, parameterName=ColorB, sortPriority=2, defaultValue=<rgba 0 1 0 1>}
  n20 = VectorParameter() {desc=End color in gradient, group=Color, parameterName=ColorC, sortPriority=3, defaultValue=<rgba 1 0 0 1>}
  n21 = NamedRerouteUsage()
  n22 = VectorParameter() {desc=Color for delta part of progress bar., group=Delta, parameterName=DeltaColor, sortPriority=20, defaultValue=<rgba 1 0 0.829 1>}
  n23 = ScalarParameter() {
   desc="Property for animating Delta color decay\r\n(from DeltaColor to DeltaColorDecay)\r\n0 - 1"
   group=Delta
   parameterName=AnimateDeltaColor
   sliderMax=1
   sortPriority=34
  }
  n24 = Max(A=NamedRerouteUsage(), B=NamedRerouteUsage())
  n25 = LinearInterpolate(
   A=MaterialFunctionCall(A=@n18.RGB, B=@n19.RGB, C=@n20.RGB, Alpha=OneMinus(None=@n21)) {materialFunction=&3ColorBlend}.Result
   B=LinearInterpolate(A=@n22.RGB, B=VectorParameter() {group=Delta, parameterName=DeltaColorDecay, sortPriority=21, defaultValue=<rgba 0 1 1 1>}.RGB, Alpha=@n23)
   Alpha=Multiply(A=NamedRerouteUsage(), B=OneMinus(None=@n24))
  ) {constAlpha=0}
  n26 = LinearInterpolate(A=@n25, B=VectorParameter() {group=Glow, parameterName=GlowColor, sortPriority=3, defaultValue=<rgba 0.1958 0 1 0.2>}.RGB, Alpha=@n17)
  n27 = NamedRerouteUsage()
  n28 = NamedRerouteUsage()
  n29 = NamedRerouteUsage()
  n30 = ScalarParameter() {
   desc="Enables inner circle fill for the circular progress bar.\r\nbool"
   group=Behind Circle
   parameterName=UseBehindCircle
   sliderMax=1
   sortPriority=1
  }
  n31 = LinearInterpolate(A=Ceil(None=@n27), B=@n27, Alpha=LinearInterpolate(A=@n28, B=Max(A=@n28, B=@n29), Alpha=@n30))
  n32 = ScalarParameter() {desc="Enables glow around progress fill part.\r\nbool", group=Glow, parameterName=UseProgressGlow, sliderMax=1, sortPriority=1}
  n33 = Max(A=NamedRerouteUsage(), B=NamedRerouteUsage())
  n34 = NamedRerouteUsage()
  n35 = ScalarParameter() {defaultValue=1, desc="Enable segments functionality.\r\nbool", group=Segments, parameterName=UseSegments, sliderMax=1, sortPriority=1}
  LinearInterpolate(
   A=LinearInterpolate(A=@n16, B=LinearInterpolate(A=@n16, B=@n26, Alpha=@n31), Alpha=@n32)
   B=@n25
   Alpha=LinearInterpolate(A=@n33, B=Multiply(A=@n33, B=@n34) {constA=1}, Alpha=@n35)
  )
  n36 = ScalarParameter() {defaultValue=20, desc=Adds space on the texture edge in pixels., group=Progress_Layout, parameterName=EdgeOffset_PX, sortPriority=10}
  n37 = Multiply(A=@n36) {constB=2}
  n38 = NamedRerouteDeclaration(None=MaterialFunctionCall!(
    Value PX=ScalarParameter() {defaultValue=20, group=Progress_Layout, parameterName=BarThickness_PX, sliderMax=1000, sortPriority=0}
    UV Pixel Size
    X or Y
   ) {materialFunction=/FortUILibrary/MaterialFunctions/Basics/MF_UI_PixelsToNormalized}."Normalized (>1px)") {name=BarThicknessNormalized, nodeColor=<rgba 0 0 0 1>}
  n39 = Add(
   A=NamedRerouteDeclaration(None=MaterialFunctionCall(Value PX=@n37) {materialFunction=&MF_UI_PixelsToNormalized}.Normalized) {name=EdgeOffsetNormalized, nodeColor=<rgba 0 0 0 1>}
   B=@n38
  )
  n40 = NamedRerouteUsage()
  n41 = ScalarParameter() {defaultValue=1, desc="Will add outline for progress bar.\r\nbool", group=Outline, parameterName=UseOutline, sliderMax=1, sortPriority=1}
  n42 = MaterialFunctionCall(Size=Subtract(A=OneMinus(None=@n39), B=LinearInterpolate(B=@n40, Alpha=@n41)), OutlineThickness=@n38) {materialFunction=&MF_UI_Circle}
  NamedRerouteDeclaration(None=@n42.SDFOutline) {name=SDF_CircleBar, nodeColor=<rgba 0 0 0 1>}
  n43 = ScalarParameter() {desc="Angle from 0 point and move clockwise.\r\n0 - 1", group=Progress_Layout, parameterName=StartAngle, sliderMax=1, sortPriority=5}
  n44 = ScalarParameter() {desc="Angle from 0 point and move anticlockwise.\r\n0 - 1", group=Progress_Layout, parameterName=EndAngle, sliderMax=1, sortPriority=6}
  n45 = ScalarParameter() {
   defaultValue=1
   desc="Will fill the progress bar clockwise.\r\nbool"
   group=Progress_Layout
   parameterName=FillClockwise
   sliderMax=1
   sortPriority=40
  }
  n46 = ScalarParameter() {
   desc="When progress value changes, it defines where change starts.\r\n0 - 1"
   group=Progress
   parameterName=ProgressStart
   sliderMax=1
   sortPriority=1
  }
  n47 = ScalarParameter() {
   defaultValue=1
   desc="When progress value changes, it defines where change ends.\r\n0 - 1"
   group=Progress
   parameterName=ProgressEnd
   sliderMax=1
   sortPriority=2
  }
  n48 = ScalarParameter() {
   defaultValue=0.5
   desc="Property for animating Delta value change (from ProgressStart to ProgressEnd)\r\n0 - 1"
   group=Delta
   parameterName=AnimateDelta
   sliderMax=1
   sortPriority=33
  }
  n49 = LinearInterpolate(A=@n46, B=@n47, Alpha=@n48)
  n50 = ScalarParameter() {defaultValue=1, desc="Change progress value only with segments chanks.\r\nbool", group=Segments, parameterName=SnapToSegments, sliderMax=1}
  n51 = LinearInterpolate(
   A=@n49
   B=MaterialFunctionCall!("Value (0-1)"=@n49, Steps=NamedRerouteUsage()) {materialFunction=/FortUILibrary/MaterialFunctions/Basics/MF_UI_StepRatio}.Stepped Ratio
   Alpha=@n50
  )
  n52 = MaterialFunctionCall!(
   UV=NamedRerouteUsage()
   "Arc Start Angle (0-1)"=@n43
   "Arc End Angle (0-1)"=@n44
   "Fill Ratio (0-1)"=OneMinus(None=@n51)
   Fill Clockwise=@n45
  ) {materialFunction=/FortUILibrary/MaterialFunctions/Basics/MF_UI_RadialArc}
  n53 = ScalarParameter() {
   desc="Enables delta functionality for progress bar.\r\nShows delta after progress on decrease progress value and on encrease shows delta value before progress.\r\nbool"
   group=Delta
   parameterName=ShowDelta
   sliderMax=1
   sortPriority=0
  }
  NamedRerouteDeclaration(None=Multiply(A=Ceil(None=@n52.Radial Fill), B=@n53)) {name=ArcDeltaMask, nodeColor=<rgba 0 0 0 1>}
  n54 = ScalarParameter() {desc="Angle from 0 point and move clockwise.\r\n0 - 1", group=Progress_Layout, parameterName=StartAngle, sliderMax=1, sortPriority=5}
  n55 = ScalarParameter() {desc="Angle from 0 point and move anticlockwise.\r\n0 - 1", group=Progress_Layout, parameterName=EndAngle, sliderMax=1, sortPriority=6}
  n56 = ScalarParameter() {
   defaultValue=1
   desc="Will fill the progress bar clockwise.\r\nbool"
   group=Progress_Layout
   parameterName=FillClockwise
   sliderMax=1
   sortPriority=40
  }
  n57 = ScalarParameter() {
   desc="When progress value changes, it defines where change starts.\r\n0 - 1"
   group=Progress
   parameterName=ProgressStart
   sliderMax=1
   sortPriority=1
  }
  n58 = ScalarParameter() {
   defaultValue=1
   desc="When progress value changes, it defines where change ends.\r\n0 - 1"
   group=Progress
   parameterName=ProgressEnd
   sliderMax=1
   sortPriority=2
  }
  n59 = ScalarParameter() {
   defaultValue=0.812
   desc="Property for animating progress value change (from ProgressStart to ProgressEnd)\r\n0 - 1"
   group=Progress
   parameterName=AnimateProgress
   sliderMax=1
   sortPriority=3
  }
  n60 = LinearInterpolate(A=@n57, B=@n58, Alpha=@n59)
  n61 = ScalarParameter() {defaultValue=1, desc="Change progress value only with segments chanks.\r\nbool", group=Segments, parameterName=SnapToSegments, sliderMax=1}
  n62 = LinearInterpolate(
   A=@n60
   B=MaterialFunctionCall("Value (0-1)"=@n60, Steps=NamedRerouteUsage()) {materialFunction=&MF_UI_StepRatio}.Stepped Ratio
   Alpha=@n61
  )
  n63 = MaterialFunctionCall(
   UV=NamedRerouteUsage()
   "Arc Start Angle (0-1)"=@n54
   "Arc End Angle (0-1)"=@n55
   "Fill Ratio (0-1)"=OneMinus(None=@n62)
   Fill Clockwise=@n56
  ) {materialFunction=&MF_UI_RadialArc}
  NamedRerouteDeclaration(None=@n63.Arc Mask) {name=ArcGradient, nodeColor=<rgba 0 0 0 1>}
  n64 = MaterialFunctionCall(Value PX=ScalarParameter() {defaultValue=4, desc=The size of gap between segements in px., group=Segments, parameterName=SeparatorThickness_PX}) {materialFunction=&MF_UI_PixelsToNormalized}
  n65 = Multiply(A=@n64."Normalized (>1px)", B=Multiply(A=NamedRerouteUsage()) {constB=0.25})
  n66 = ScalarParameter() {defaultValue=0.01, group=Progress_Layout, parameterName=SeparatorSharpness, sortPriority=42}
  n67 = ScalarParameter() {defaultValue=8, desc=Number of segments in progress bar, group=Segments, parameterName=SegmentsCount, sliderMin=1, sortPriority=2}
  n68 = Multiply(A=NamedRerouteUsage(), B=NamedRerouteDeclaration(None=Floor(None=@n67)) {name=SegmentsCount, nodeColor=<rgba 0 0 0 1>})
  n69 = Constant() {r=1}
  n70 = ConstantBiasScale(None=NamedRerouteUsage()) {bias=-0.5, scale=2}
  n71 = Saturate(None=Multiply(
    A=MaterialFunctionCall!(Value=@n68, Scale=@n69, Height) {materialFunction=/FortUILibrary/MaterialFunctions/Basics/MF_UI_PingPong}.Result
    B=Length(None=@n70) {is=class MaterialExpressionLength}
   ))
  NamedRerouteDeclaration(None=SmoothStep(Min=Subtract(A=@n65, B=@n66), Max=Add(A=@n65, B=@n66), Value=Subtract(A=@n71, B=@n64."Normalized (>1px)")) {constMax=0.01}) {name=SegmentsMask, nodeColor=<rgba 0 0 0 1>}
  n72 = NamedRerouteUsage()
  n73 = Multiply(A=ScalarParameter() {defaultValue=5, group=Outline, parameterName=Outline_PX, sortPriority=2}) {constB=2}
  n74 = Add(
   A=NamedRerouteDeclaration(None=MaterialFunctionCall(Value PX=@n73) {materialFunction=&MF_UI_PixelsToNormalized}.Normalized) {name=OutlineThicknessNormalized, nodeColor=<rgba 0 0 0 1>}
   B=NamedRerouteUsage()
  )
  n75 = MaterialFunctionCall(Size=Subtract(A=OneMinus(None=@n74), B=NamedRerouteUsage()), OutlineThickness=@n74) {materialFunction=&MF_UI_Circle}
  n76 = ScalarParameter() {defaultValue=1, desc="Will add outline for progress bar.\r\nbool", group=Outline, parameterName=UseOutline, sliderMax=1, sortPriority=1}
  NamedRerouteDeclaration(None=MaterialFunctionCall(DistanceField=LinearInterpolate(A=@n72, B=@n75.SDFOutline, Alpha=@n76)) {materialFunction=&MF_UI_AntiAliasing}.Inner) {name=CircleBarMask, nodeColor=<rgba 0 0 0 1>}
  NamedRerouteDeclaration(None=Ceil(None=@n63.Radial Fill)) {name=ArcProgressMask, nodeColor=<rgba 0 0 0 1>}
  n77 = ScalarParameter() {defaultValue=0.1, group=Glow, parameterName=GlowIntensity, sortPriority=4}
  n78 = ScalarParameter() {
   desc="Will show rounded cap for start and end edge of the progress bar. Not support delta functionality and snap to segments.\r\nbool"
   group=Progress_Layout
   parameterName=UseRoundedProgress
   sliderMax=1
   sortPriority=2
  }
  n79 = LinearInterpolate(
   A=Min(A=NamedRerouteUsage(), B=NamedRerouteUsage())
   B=NamedRerouteUsage()
   Alpha=LinearInterpolate(A=NamedRerouteUsage(), B=NamedRerouteUsage(), Alpha=@n78) {constA=1}
  )
  NamedRerouteDeclaration(None=OneMinus(None=SmoothStep(Max=@n77, Value=@n79) {constMax=0.15})) {name=GlowMask, nodeColor=<rgba 0 0 0 1>}
  NamedRerouteDeclaration(None=@n51) {name=DeltaProgress, nodeColor=<rgba 0 0 0 1>}
  n80 = ScalarParameter() {desc="Angle from 0 point and move anticlockwise.\r\n0 - 1", group=Progress_Layout, parameterName=EndAngle, sliderMax=1, sortPriority=6}
  n81 = OneMinus(None=@n80)
  n82 = ScalarParameter() {desc="Angle from 0 point and move clockwise.\r\n0 - 1", group=Progress_Layout, parameterName=StartAngle, sliderMax=1, sortPriority=5}
  n83 = ScalarParameter() {
   defaultValue=1
   desc="Will fill the progress bar clockwise.\r\nbool"
   group=Progress_Layout
   parameterName=FillClockwise
   sliderMax=1
   sortPriority=40
  }
  n84 = LinearInterpolate(A=@n81, B=@n82, Alpha=@n83)
  n85 = ScalarParameter() {
   desc="Will show rounded cap for start and end edge of the progress bar. Not support delta functionality and snap to segments.\r\nbool"
   group=Progress_Layout
   parameterName=UseRoundedProgress
   sliderMax=1
   sortPriority=2
  }
  n86 = LinearInterpolate(A=NamedRerouteUsage(), Alpha=@n85)
  n87 = MaterialFunctionCall(UVs=NamedRerouteUsage(), "Rotation Angle (0-1)"=Subtract(A=OneMinus(None=@n84), B=@n86) {constB=0}) {materialFunction=&CustomRotator}
  n88 = ComponentMask(None=@n87.Rotated Values) {r=true}
  n89 = Add(A=ComponentMask(None=@n87.Rotated Values) {g=true}, B=NamedRerouteUsage()) {constB=0.47}
  NamedRerouteDeclaration(None=MaterialFunctionCall(UVs=AppendVector(A=@n88, B=@n89), Size=NamedRerouteUsage()) {materialFunction=&MF_UI_Circle}.SDF) {name=SDF_CircleStart, nodeColor=<rgba 0 0 0 1>}
  n90 = LinearInterpolate(A=@n82, B=@n81, Alpha=@n83)
  n91 = NamedRerouteUsage()
  n92 = MaterialFunctionCall(UVs=NamedRerouteUsage(), "Rotation Angle (0-1)"=OneMinus(None=LinearInterpolate(A=@n84, B=@n90, Alpha=@n91))) {materialFunction=&CustomRotator}
  n93 = ComponentMask(None=@n92.Rotated Values) {r=true}
  n94 = Add(A=ComponentMask(None=@n92.Rotated Values) {g=true}, B=NamedRerouteUsage()) {constB=0.47}
  NamedRerouteDeclaration(None=MaterialFunctionCall(UVs=AppendVector(A=@n93, B=@n94), Size=NamedRerouteUsage()) {materialFunction=&MF_UI_Circle}.SDF) {name=SDF_CircleDeltaEnd, nodeColor=<rgba 0 0 0 1>}
  n95 = ScalarParameter() {
   desc="Will show rounded cap for start and end edge of the progress bar. Not support delta functionality and snap to segments.\r\nbool"
   group=Progress_Layout
   parameterName=UseRoundedProgress
   sliderMax=1
   sortPriority=2
  }
  n96 = LinearInterpolate(A=LinearInterpolate(B=NamedRerouteUsage(), Alpha=@n95) {constA=1}, B=NamedRerouteUsage(), Alpha=NamedRerouteUsage())
  n97 = ScalarParameter() {
   desc="Enables delta functionality for progress bar.\r\nShows delta after progress on decrease progress value and on encrease shows delta value before progress.\r\nbool"
   group=Delta
   parameterName=ShowDelta
   sliderMax=1
   sortPriority=0
  }
  NamedRerouteDeclaration(None=Multiply(
    A=MaterialFunctionCall(DistanceField=@n96) {materialFunction=&MF_UI_AntiAliasing}.Inner
    B=@n97
   )) {name=DeltaMask, nodeColor=<rgba 0 0 0 1>}
  n98 = NamedRerouteUsage()
  NamedRerouteDeclaration(None=Multiply(A=@n42.Fill, B=OneMinus(None=@n98))) {name=BehindCircle_Inside, nodeColor=<rgba 0 0 0 1>}
  n99 = LinearInterpolate(A=@n84, B=@n90, Alpha=NamedRerouteUsage())
  n100 = MaterialFunctionCall(UVs=NamedRerouteUsage(), "Rotation Angle (0-1)"=Add(A=OneMinus(None=@n99), B=@n86) {constB=0}) {materialFunction=&CustomRotator}
  n101 = ComponentMask(None=@n100.Rotated Values) {r=true}
  n102 = Add(A=ComponentMask(None=@n100.Rotated Values) {g=true}, B=NamedRerouteUsage()) {constB=0.47}
  NamedRerouteDeclaration(None=MaterialFunctionCall(UVs=AppendVector(A=@n101, B=@n102), Size=NamedRerouteUsage()) {materialFunction=&MF_UI_Circle}.SDF) {name=SDF_CircleEnd, nodeColor=<rgba 0 0 0 1>}
  NamedRerouteDeclaration(None=MaterialFunctionCall() {materialFunction=&GetUserInterfaceUV}.Normalized UV) {name=UVs, nodeColor=<rgba 0 0 0 1>}
  n103 = ScalarParameter() {
   desc="Rotate the start point of the progress bar.\r\n0 - top\r\n0.25 - left\r\n0.5 - bottom\r\n0.75 - right"
   group=Progress_Layout
   parameterName=RotationAngle
  }
  NamedRerouteDeclaration(None=MaterialFunctionCall(UVs=NamedRerouteUsage(), "Rotation Angle (0-1)"=@n103) {materialFunction=&CustomRotator}.Rotated Values) {name=RotatorUVs, nodeColor=<rgba 0 0 0 1>}
  NamedRerouteDeclaration(None=Ceil(None=@n63.Arc Mask)) {name=ArcEmptyPartMask, nodeColor=<rgba 0 0 0 1>}
  n104 = ScalarParameter() {
   desc="Will show rounded cap for start and end edge of the progress bar. Not support delta functionality and snap to segments.\r\nbool"
   group=Progress_Layout
   parameterName=UseRoundedProgress
   sliderMax=1
   sortPriority=2
  }
  n105 = LinearInterpolate(B=Min(A=NamedRerouteUsage(), B=NamedRerouteUsage()), Alpha=@n104) {constA=1}
  n106 = NamedRerouteUsage()
  n107 = NamedRerouteUsage()
  NamedRerouteDeclaration(None=MaterialFunctionCall(DistanceField=LinearInterpolate(A=@n105, B=@n107, Alpha=@n106)) {materialFunction=&MF_UI_AntiAliasing}.Inner) {name=ProgressMask, nodeColor=<rgba 0 0 0 1>}
  n108 = NamedRerouteUsage()
  n109 = NamedRerouteUsage()
  NamedRerouteDeclaration(None=Divide(A=Multiply(A=@n108) {constB=0.53}, B=Multiply(A=@n109) {constB=6.2832})) {name=HalfCapArcAngle, nodeColor=<rgba 0 0 0 1>}
  n110 = NamedRerouteUsage()
  n111 = NamedRerouteUsage()
  n112 = ScalarParameter() {defaultValue=1, desc="Will add outline for progress bar.\r\nbool", group=Outline, parameterName=UseOutline, sliderMax=1, sortPriority=1}
  n113 = Multiply(A=LinearInterpolate(A=@n110, B=Add(A=@n110, B=@n111), Alpha=@n112)) {constB=0.5}
  n114 = NamedRerouteUsage()
  NamedRerouteDeclaration(None=Subtract(A=Subtract(B=@n113) {constA=0.5}, B=Multiply(A=@n114) {constB=0.5})) {name=CapOffset, nodeColor=<rgba 0 0 0 1>}
  n115 = NamedRerouteUsage()
  n116 = Subtract(A=NamedRerouteUsage(), B=@n115)
  NamedRerouteDeclaration(None=Subtract(A=Step(Y=@n115, X=@n63.Radial Fill), B=Step(Y=@n116, X=@n63.Radial Fill))) {name=ArcProgressMinusCapsMask, nodeColor=<rgba 0 0 0 1>}
  NamedRerouteDeclaration(None=@n62) {name=Progress, nodeColor=<rgba 0 0 0 1>}
 ]
}
