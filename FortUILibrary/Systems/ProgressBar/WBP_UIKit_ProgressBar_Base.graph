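# Base UI kit progress bar widget: a single Image drawn by a linear or circular material instance, with optional segments, outline, glow and rounded fill.
# PreConstruct creates the dynamic material and pushes the style variables into its parameters.
# SetProgress jumps to a value, while transitions play duration-scaled animations on material parameters, using delta, decay or caller-supplied animations.
# With delta enabled, decreases swap start and end and play in reverse.
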WidgetBlueprint /FortUILibrary/Systems/ProgressBar/WBP_UIKit_ProgressBar_Base
animations [
 Anim_Progress_Base {
  binding_count=1
  length_seconds=1
  object=&self:Anim_Progress_Base
  sequence {
   bindings [ProgressBar {
     id=#1
     possessed_class=Image
     tracks [{
       type=MovieSceneWidgetMaterialTrack
       sections [{
         type=MovieSceneParameterSection
         end_frame=100
         start_frame=0
         channels [AnimateProgress {type=MovieSceneScriptingFloatChannel, keys [[0,0],[100,1]]}]
        }]
      }]
    }]
   display_rate {numerator=100}
  }
 }
 Anim_ProgressWithDelta_Base {
  binding_count=1
  length_seconds=1
  object=&self:Anim_ProgressWithDelta_Base
  sequence {
   bindings [ProgressBar {
     id=#1
     possessed_class=Image
     tracks [{
       type=MovieSceneWidgetMaterialTrack
       sections [{
         type=MovieSceneParameterSection
         end_frame=240
         start_frame=0
         channels [
          AnimateDelta {type=MovieSceneScriptingFloatChannel, keys [[0,0],[60,1]]}
          AnimateProgress {type=MovieSceneScriptingFloatChannel, keys [[120,0],[240,1]]}
         ]
        }]
      }]
    }]
   display_rate {numerator=240}
  }
 }
 Anim_ProgressWithDeltaDecay_Base {
  binding_count=1
  length_seconds=1
  object=&self:Anim_ProgressWithDeltaDecay_Base
  sequence {
   bindings [ProgressBar {
     id=#1
     possessed_class=Image
     tracks [{
       type=MovieSceneWidgetMaterialTrack
       sections [{
         type=MovieSceneParameterSection
         end_frame=100
         start_frame=0
         channels [
          AnimateDelta {type=MovieSceneScriptingFloatChannel, keys [[0,0],[10,1]]}
          AnimateDeltaColor {type=MovieSceneScriptingFloatChannel, keys [[0,0],[80,1]]}
          AnimateProgress {type=MovieSceneScriptingFloatChannel, keys [[75,0],[100,1]]}
         ]
        }]
      }]
    }]
   display_rate {numerator=100}
  }
 }
 Anim_ProgressWithDelta_Ease {
  binding_count=1
  length_seconds=1
  object=&self:Anim_ProgressWithDelta_Ease
  sequence {
   bindings [ProgressBar {
     id=#1
     possessed_class=Image
     tracks [{
       type=MovieSceneWidgetMaterialTrack
       sections [{
         type=MovieSceneParameterSection
         end_frame=240
         start_frame=0
         channels [
          AnimateDelta {type=MovieSceneScriptingFloatChannel, keys [[0,0],[167,1]]}
          AnimateProgress {type=MovieSceneScriptingFloatChannel, default=1, keys [[73,0],[239,1]]}
         ]
        }]
      }]
    }]
   display_rate {numerator=240}
  }
 }
]
graphs [
 AnimateProgressChange {nodes [
   n1 = AnimateProgressChange!() -> (then, CurrentProgress:double=~, TargetProgress:double=~, TimeDuration:double=~, ProgressAnimation:Widget Animation Object Reference) {is=title K2Node_FunctionEntry}
   > n2 = Get!(execute:Exec) -> (then:Exec, else:Exec, CurrentProgressAnimation:Widget Animation Object Reference)
   > n3 = UnbindAllFromAnimationFinished!(execute, self:User Widget Object Reference, Animation:Widget Animation Object Reference=@n2) -> (then)
   n4 = Branch(execute=[@n3.then, @n2.else], Condition=Get UseDeltaBar() -> (UseDeltaBar:bool))
   > n5 = Branch(Condition="float < float"(A=@n1, B=@n1.TargetProgress))
   n6 = Get ProgressBarMID() -> (ProgressBarMID:Material Instance Dynamic Object Reference)
   > SetScalarParameterValue(self=@n6, ParameterName=ProgressStart, Value=Get CurrentProgress() -> (CurrentProgress:double))
   > SetScalarParameterValue(self=@n6, ParameterName=ProgressEnd, Value=Get TargetProgress() -> (TargetProgress:double))
   > n7 = Branch(Condition=Get UseDeltaDecay() -> (UseDeltaDecay:bool))
   > SetVectorParameterValue(self=@n6, ParameterName=DeltaColor, Value=Get DeltaColor() -> (DeltaColor:Linear Color Structure))
   > n8 = SetVectorParameterValue(self=@n6, ParameterName=DeltaColorDecay, Value=Get DeltaDecayColor() -> (DeltaDecayColor:Linear Color Structure))
   n9 = Get ProgressAnimation() -> (ProgressAnimation:Widget Animation Object Reference)
   n10 = Select(
    Option 0:Widget Animation Object Reference=Select(
     Option 0=Get Anim_ProgressWithDelta_Base() -> (Anim_ProgressWithDelta_Base:Widget Animation Object Reference)
     Option 1=Get Anim_ProgressWithDeltaDecay_Base() -> (Anim_ProgressWithDeltaDecay_Base:Widget Animation Object Reference)
     Index=Get UseDeltaDecay()
    )
    Option 1:Widget Animation Object Reference=@n9
    Index=IsValid(Object=@n9)
   ) -> (ReturnValue:Widget Animation Object Reference)
   n11 = Set CurrentProgressAnimation(execute=[@n8.then, @n7.else], CurrentProgressAnimation:Widget Animation Object Reference=@n10) -> (Output_Get:Widget Animation Object Reference)
   > n12 = Branch(Condition=NOT Boolean(A=Get Is Design Time() -> (Is Design Time:bool)))
   > BindToAnimationFinished(Animation=@n11, Delegate=Create Event())
   > PlayAnimationForward(
    InAnimation=Get CurrentProgressAnimation() -> (CurrentProgressAnimation:Widget Animation Object Reference)
    PlaybackSpeed=SafeDivide(A=1, B=Get TimeDuration() -> (TimeDuration:double))
   )
   SetScalarParameterValue(execute=@n12.else, self=Get ProgressBarMID(), ParameterName=AnimateDelta, Value=1)
   > SetScalarParameterValue(self=Get ProgressBarMID(), ParameterName=AnimateProgress, Value=0)
   n13 = Get ProgressBarMID()
   SetScalarParameterValue(execute=@n5.else, self=@n13, ParameterName=ProgressStart, Value=Get TargetProgress())
   > SetScalarParameterValue(self=@n13, ParameterName=ProgressEnd, Value=Get CurrentProgress())
   > n14 = Branch(Condition=Get UseDeltaDecay())
   > SetVectorParameterValue(self=@n13, ParameterName=DeltaColor, Value=Get DeltaDecayColor())
   > n15 = SetVectorParameterValue(self=@n13, ParameterName=DeltaColorDecay, Value=Get DeltaColor())
   n16 = Set CurrentProgressAnimation(execute=[@n15.then, @n14.else], CurrentProgressAnimation=@n10)
   > n17 = Branch(Condition=NOT Boolean(A=Get Is Design Time()))
   > BindToAnimationFinished(Animation=@n16, Delegate=Create Event())
   > PlayAnimationReverse(InAnimation=Get CurrentProgressAnimation(), PlaybackSpeed=SafeDivide(A=1, B=Get TimeDuration()))
   SetScalarParameterValue(execute=@n17.else, self=Get ProgressBarMID(), ParameterName=AnimateDelta, Value=0)
   > SetScalarParameterValue(self=Get ProgressBarMID(), ParameterName=AnimateProgress, Value=1)
   n18 = Get ProgressBarMID()
   SetScalarParameterValue(execute=@n4.else, self=@n18, ParameterName=ProgressStart, Value=Get CurrentProgress())
   > SetScalarParameterValue(self=@n18, ParameterName=ProgressEnd, Value=Get TargetProgress())
   > Branch(Condition=NOT Boolean(A=Get Is Design Time()))
   n19 = Get ProgressAnimation()
   > n20 = Set CurrentProgressAnimation(CurrentProgressAnimation=Select(Option 0=Get Anim_Progress_Base() -> (Anim_Progress_Base:Widget Animation Object Reference), Option 1=@n19, Index=IsValid(Object=@n19)))
   > BindToAnimationFinished(Animation=@n20, Delegate=Create Event())
   > PlayAnimationForward(InAnimation=Get CurrentProgressAnimation(), PlaybackSpeed=SafeDivide(A=1, B=Get TimeDuration()))
  ]}
 EventGraph {nodes [
   n1 = Event PreConstruct()
   > Set Is Design Time(Is Design Time:bool=@n1.IsDesignTime) -> (Output_Get:bool)
   > n2 = Sequence()
   > SetType!(execute, self, IsCircular:bool=Get UseCircular() -> (UseCircular:bool), IsRounded:bool=Get UseRoundedProgress() -> (UseRoundedProgress:bool)) -> (then)
   > InitializeProgressBar()
   Branch(execute=@n2.then_1, Condition=@n1.IsDesignTime)
   > SetProgress!(execute, self, Progress:double=Get ProgressValue() -> (ProgressValue:double)) -> (then)
  ]}
 InitializeProgressBar {nodes [
   InitializeProgressBar() {is=title K2Node_FunctionEntry}
   > n1 = Sequence!(execute) -> (then_0, then_1, then_2:Exec)
   > SetFillColors!(
    execute
    self
    ColorA:Linear Color Structure=Get ColorA() -> (ColorA:Linear Color Structure)
    ColorB:Linear Color Structure=Get ColorB() -> (ColorB:Linear Color Structure)
    ColorC:Linear Color Structure=Get ColorC() -> (ColorC:Linear Color Structure)
   ) -> (then)
   > SetDeltaColor!(execute, self, DeltaStartColor:Linear Color Structure=Get DeltaColor(), DeltaEndColor:Linear Color Structure=Get DeltaDecayColor()) -> (then)
   > SetEmptyColor!(execute, self, EmptyColor:Linear Color Structure=Get EmptyColor() -> (EmptyColor:Linear Color Structure)) -> (then)
   SetOutline!(execute=@n1.then_1, self, UseOutline:bool=Get UseOutline() -> (UseOutline:bool), OutlinePX:double=Get Outline_PX() -> (Outline_PX:double)) -> (then)
   > SetSharpness!(
    execute
    self
    Sharpness:double=SelectFloat(A=Get SDF_Sharpness() -> (SDF_Sharpness:double), B=Get LinearSDF_Sharpness() -> (LinearSDF_Sharpness:double), bPickA=Get UseCircular())
   ) -> (then)
   > n2 = Branch(Condition=Get UseCircular())
   > SetCircularParameters!(
    execute
    self
    FillClokcwise:bool=Get FillClockwise() -> (FillClockwise:bool)
    Thickness_PX:double=Get Thickness_PX() -> (Thickness_PX:double)
    StartAngle:double=Get StartAngle() -> (StartAngle:double)
    EndAngle:double=Get EndAngle() -> (EndAngle:double)
    RotationAngle:double=Get RotationAngle() -> (RotationAngle:double)
    EdgeOffset_PX:double=Get EdgeOffset_PX() -> (EdgeOffset_PX:double)
    UseBehindCircle:bool=Get UseBehindCircle() -> (UseBehindCircle:bool)
    BehindCircleColor:Linear Color Structure=Get BehindCircleColor() -> (BehindCircleColor:Linear Color Structure)
   ) -> (then)
   n3 = Get Width() -> (Width:double)
   > SetSize!(execute, self, InWidth:double=@n3, InHeight:double=@n3) -> (then)
   > SetProgressGlow!(
    execute
    self
    UseGlow:bool=Get UseGlow() -> (UseGlow:bool)
    Intensity:double=Get GlowIntensity() -> (GlowIntensity:double)
    UseGlowColor:bool=Get UseGlowColor() -> (UseGlowColor:bool)
    Color:Linear Color Structure=Get GlowColor() -> (GlowColor:Linear Color Structure)
   ) -> (then)
   SetLiniarParameters!(execute=@n2.else, self, CornerRoundness:double=Get CornerRoundness() -> (CornerRoundness:double)) -> (then)
   > SetSize(InWidth=Get Width(), InHeight=Get Height() -> (Height:double))
   SetSegments!(
    execute=@n1.then_2
    self
    UseSegments:bool=Get UseSegments() -> (UseSegments:bool)
    SnapToSegments:bool=Get SnapToSegments() -> (SnapToSegments:bool)
    SegmentsCount:double="To Float (Integer)"(InInt=Get SegmentsCount() -> (SegmentsCount:int=0))
    SeparatorThickness:double=Get SeparatorThickness_PX() -> (SeparatorThickness_PX:double)
   ) -> (then)
   > SetDeltaVisibility!(execute, self, ShowDelta:bool=Get UseDeltaBar()) -> (then)
  ]}
 OnAnimateProgressChangeFinished {nodes [
   OnAnimateProgressChangeFinished() {is=title K2Node_FunctionEntry}
   > Call On Progress Transition Finished!(execute:Exec, self:WBP UIKit Progress Bar Base Object Reference) -> (then:Exec) {is=title K2Node_CallDelegate}
   > n1 = Get()
   > UnbindAllFromAnimationFinished(Animation=@n1)
   > Set CurrentProgressAnimation()
  ]}
 OnProgressTransitionFinished {nodes [OnProgressTransitionFinished() {is=title K2Node_FunctionEntry}]}
 SetCircularParameters {nodes [
   n1 = SetCircularParameters!() -> (
    then
    FillClokcwise:bool
    Thickness_PX:double=~
    StartAngle:double=~
    EndAngle:double=~
    RotationAngle:double=~
    EdgeOffset_PX:double=~
    UseBehindCircle:bool=~
    BehindCircleColor:Linear Color Structure
   ) {is=title K2Node_FunctionEntry}
   n2 = Get ProgressBarMID()
   > SetScalarParameterValue(self=@n2, ParameterName=FillClockwise, Value="To Float (Boolean)"(InBool=@n1))
   > SetScalarParameterValue(self=@n2, ParameterName=BarThickness_PX, Value=@n1.Thickness_PX)
   > SetScalarParameterValue(self=@n2, ParameterName=StartAngle, Value=@n1.StartAngle)
   > SetScalarParameterValue(self=@n2, ParameterName=EndAngle, Value=@n1.EndAngle)
   > SetScalarParameterValue(self=@n2, ParameterName=RotationAngle, Value=@n1.RotationAngle)
   > SetScalarParameterValue(self=@n2, ParameterName=EdgeOffset_PX, Value=@n1.EdgeOffset_PX)
   > SetScalarParameterValue(self=@n2, ParameterName=UseBehindCircle, Value="To Float (Boolean)"(InBool=@n1.UseBehindCircle))
   > SetVectorParameterValue(self=@n2, ParameterName=BehindCircleColor, Value=@n1.BehindCircleColor)
  ]}
 SetDeltaColor {nodes [
   n1 = SetDeltaColor!() -> (then, DeltaStartColor:Linear Color Structure, DeltaEndColor:Linear Color Structure) {is=title K2Node_FunctionEntry}
   n2 = Get ProgressBarMID()
   > SetVectorParameterValue(self=@n2, ParameterName=DeltaColor, Value=@n1)
   > SetVectorParameterValue(self=@n2, ParameterName=DeltaColorDecay, Value=@n1.DeltaEndColor)
  ]}
 SetDeltaVisibility {nodes [
   n1 = SetDeltaVisibility!() -> (then, ShowDelta:bool=~) {is=title K2Node_FunctionEntry}
   > SetScalarParameterValue(self=Get ProgressBarMID(), ParameterName=ShowDelta, Value="To Float (Boolean)"(InBool=@n1))
  ]}
 SetDurationTime {nodes [
   n1 = SetDurationTime!() -> (then, DurationTime:double=~) {is=title K2Node_FunctionEntry}
   > Set DurationTime(DurationTime:double=@n1) -> (Output_Get:double)
  ]}
 SetEmptyColor {nodes [
   n1 = SetEmptyColor!() -> (then, EmptyColor:Linear Color Structure) {is=title K2Node_FunctionEntry}
   > SetVectorParameterValue(self=Get ProgressBarMID(), ParameterName=EmptyColor, Value=@n1)
  ]}
 SetFillColors {nodes [
   n1 = SetFillColors!() -> (then, ColorA:Linear Color Structure, ColorB:Linear Color Structure, ColorC:Linear Color Structure) {is=title K2Node_FunctionEntry}
   n2 = Get ProgressBarMID()
   > SetVectorParameterValue(self=@n2, ParameterName=ColorA, Value=@n1)
   > SetVectorParameterValue(self=@n2, ParameterName=ColorB, Value=@n1.ColorB)
   > SetVectorParameterValue(self=@n2, ParameterName=ColorC, Value=@n1.ColorC)
  ]}
 SetLiniarParameters {nodes [
   n1 = SetLiniarParameters!() -> (then, CornerRoundness:double=~) {is=title K2Node_FunctionEntry}
   > SetScalarParameterValue(self=Get ProgressBarMID(), ParameterName=CornerRoundness, Value=@n1)
  ]}
 SetOutline {nodes [
   n1 = SetOutline!() -> (then, UseOutline:bool=~, OutlinePX:double=~) {is=title K2Node_FunctionEntry}
   n2 = Get ProgressBarMID()
   > SetScalarParameterValue(self=@n2, ParameterName=UseOutline, Value="To Float (Boolean)"(InBool=@n1))
   > SetScalarParameterValue(self=@n2, ParameterName=Outline_PX, Value=@n1.OutlinePX)
  ]}
 SetProgress {nodes [
   n1 = SetProgress!() -> (then, Progress:double=~) {is=title K2Node_FunctionEntry}
   > n2 = Set ProgressValue(ProgressValue:double=@n1) -> (Output_Get:double)
   n3 = Get ProgressBarMID()
   > SetScalarParameterValue(self=@n3, ParameterName=ProgressEnd, Value=@n2)
   > SetScalarParameterValue(self=@n3, ParameterName=AnimateProgress, Value=1)
  ]}
 SetProgressGlow {nodes [
   n1 = SetProgressGlow!() -> (then, UseGlow:bool=~, Intensity:double=~, UseGlowColor:bool=~, Color:Linear Color Structure) {is=title K2Node_FunctionEntry}
   n2 = Get ProgressBarMID()
   > SetScalarParameterValue(self=@n2, ParameterName=UseProgressGlow, Value="To Float (Boolean)"(InBool=@n1))
   > SetScalarParameterValue(self=@n2, ParameterName=GlowIntensity, Value=@n1.Intensity)
   > SetScalarParameterValue(self=@n2, ParameterName=UseGlowColor, Value="To Float (Boolean)"(InBool=@n1.UseGlowColor))
   > SetVectorParameterValue(self=@n2, ParameterName=GlowColor, Value=@n1.Color)
  ]}
 SetProgressTransition {nodes [
   n1 = SetProgressTransition!() -> (then, Progress:double=~) {is=title K2Node_FunctionEntry}
   > AnimateProgressChange!(
    execute
    self
    CurrentProgress:double=Get ProgressValue()
    TargetProgress:double=@n1
    TimeDuration:double=Get DurationTime() -> (DurationTime:double)
    ProgressAnimation:Widget Animation Object Reference
   ) -> (then) {is=title K2Node_CallFunction}
   > Set ProgressValue(ProgressValue=@n1)
  ]}
 SetProgressTransitionWithCustomAnimation {nodes [
   n1 = SetProgressTransitionWithCustomAnimation!() -> (then, Progress:double=~, ProgressAnimation:Widget Animation Object Reference) {is=title K2Node_FunctionEntry}
   > AnimateProgressChange(CurrentProgress=Get ProgressValue(), TargetProgress=@n1, TimeDuration=Get DurationTime(), ProgressAnimation=@n1.ProgressAnimation)
   > Set ProgressValue(ProgressValue=@n1)
  ]}
 SetSegments {nodes [
   n1 = SetSegments!() -> (then, UseSegments:bool=~, SnapToSegments:bool=~, SegmentsCount:double=~, SeparatorThickness:double=~) {is=title K2Node_FunctionEntry}
   n2 = Get ProgressBarMID()
   > SetScalarParameterValue(self=@n2, ParameterName=UseSegments, Value="To Float (Boolean)"(InBool=@n1))
   > SetScalarParameterValue(self=@n2, ParameterName=SnapToSegments, Value="To Float (Boolean)"(InBool=@n1.SnapToSegments))
   > SetScalarParameterValue(self=@n2, ParameterName=SegmentsCount, Value=@n1.SegmentsCount)
   > SetScalarParameterValue(self=@n2, ParameterName=SeparatorThickness_PX, Value=@n1.SeparatorThickness)
  ]}
 SetSharpness {nodes [
   n1 = SetSharpness!() -> (then, Sharpness:double=~) {is=title K2Node_FunctionEntry}
   > n2 = Branch(Condition=Get UseCircular())
   n3 = Get ProgressBarMID()
   > SetScalarParameterValue(self=@n3, ParameterName=SDF_Sharpness, Value=@n1)
   SetScalarParameterValue(execute=@n2.else, self=@n3, ParameterName=LinearSDF_Sharpness, Value=@n1)
  ]}
 SetSize {nodes [
   n1 = SetSize!() -> (then, InWidth:double=~, InHeight:double=~) {is=title K2Node_FunctionEntry}
   > SetDesiredSizeOverride!(
    execute
    self:Image Object Reference=Get ProgressBar() -> (ProgressBar:Image Object Reference)
    DesiredSize_X:double=@n1
    DesiredSize_Y:double=@n1.InHeight
   ) -> (then)
  ]}
 SetThickness {nodes [
   n1 = SetThickness!() -> (then, Thickness_PX:double=~) {is=title K2Node_FunctionEntry}
   > n2 = Set Thickness_PX(Thickness_PX:double=@n1) -> (Output_Get:double)
   > SetScalarParameterValue(self=Get ProgressBarMID(), ParameterName=BarThickness_PX, Value=@n2)
  ]}
 SetType {nodes [
   n1 = SetType!() -> (then, IsCircular:bool=~, IsRounded:bool=~) {is=title K2Node_FunctionEntry}
   > n2 = Branch(Condition=@n1)
   n3 = Get ProgressBar()
   > SetBrushFromMaterial(self=@n3, Material=Get CircularMaterial() -> (CircularMaterial:Material Instance Object Reference))
   > n4 = GetDynamicMaterial(self=@n3)
   > n5 = Set ProgressBarMID(ProgressBarMID:Material Instance Dynamic Object Reference=@n4) -> (Output_Get:Material Instance Dynamic Object Reference)
   n6 = Get ProgressBar()
   n7 = GetDynamicMaterial(execute=@n9.then, self=@n6)
   > n8 = Set ProgressBarMID(ProgressBarMID=@n7)
   SetScalarParameterValue(execute=[@n8.then, @n5.then], self=[@n8, @n5], ParameterName=UseRoundedProgress, Value="To Float (Boolean)"(InBool=@n1.IsRounded))
   n9 = SetBrushFromMaterial(execute=@n2.else, self=@n6, Material=Get LinearMaterial() -> (LinearMaterial:Material Instance Object Reference))
  ]}
]
variables [
 BehindCircleColor {type=LinearColor, default=<rgba 0.0334 0.6927 0 0.354>}
 CircularMaterial {type=MaterialInstance, default=/FortUILibrary/Systems/ProgressBar/MaterialInstances/MI_UI_CircularProgressBar}
 ColorA {type=LinearColor, default=<rgba 1 0 1 1>}
 ColorB {type=LinearColor, default=<rgba 1 0 1 1>}
 ColorC {type=LinearColor, default=<rgba 1 0 1 1>}
 CornerRoundness {type=number}
 CurrentProgressAnimation {type=WidgetAnimation, default=null}
 DeltaColor {type=LinearColor, default=<rgba 0 0.3087 0.6823 1>}
 DeltaDecayColor {type=LinearColor, default=<rgba 0.1691 0.917 1 1>}
 DurationTime {type=number}
 EdgeOffset_PX {type=number, default=50}
 EmptyColor {type=LinearColor, default=<rgba 0.4809 0 0.4809 0.5>}
 EndAngle {type=number}
 FillClockwise {type=boolean, default=true}
 GlowColor {type=LinearColor, default=<rgba 1 1 1 0.304>}
 GlowIntensity {type=number, default=0.15}
 Height {type=number, default=20}
 Is Design Time {type=boolean}
 LinearMaterial {type=MaterialInstance, default=/FortUILibrary/Systems/ProgressBar/MaterialInstances/MI_UI_LinearProgressBar}
 LinearSDF_Sharpness {type=number, default=0.05}
 OnProgressTransitionFinished {}
 Outline_PX {type=number, default=5}
 ProgressBarMID {type=MaterialInstanceDynamic}
 ProgressValue {type=number, default=0.25}
 RotationAngle {type=number}
 SDF_Sharpness {type=number, default=0.005}
 SegmentsCount {type=integer, default=8}
 SeparatorThickness_PX {type=number, default=2}
 SnapToSegments {type=boolean}
 StartAngle {type=number}
 Thickness_PX {type=number, default=20}
 UseBehindCircle {type=boolean}
 UseCircular {type=boolean}
 UseDeltaBar {type=boolean, default=true}
 UseDeltaDecay {type=boolean}
 UseGlow {type=boolean}
 UseGlowColor {type=boolean, default=true}
 UseOutline {type=boolean}
 UseRoundedProgress {type=boolean}
 UseSegments {type=boolean}
 Width {type=number, default=220}
]
widgets {
 info {inheritedWidgetCount=0, parentClass=&UserWidget, rootWidgetClass=&Image, widgetCount=1}
 widgets [ProgressBar {widgetClassPath=&Image, bIsVariable=true, widget=ProgressBar, properties {visibility=SelfHitTestInvisible}}]
}
